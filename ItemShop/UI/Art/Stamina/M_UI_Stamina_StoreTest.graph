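# UI-domain test material for an item shop stamina graphic, switching between purple and base stamina textures.
# A static switch, defaulting to purple, picks the RGB of either texture, with no link to a material output shown.
# Both branches scale normalized UI UVs about the center by (2, 1).
# The X scale divides pixel height by itself, so it does not adapt to widget aspect ratio.
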Material /ItemShop/UI/Art/Stamina/M_UI_Stamina_StoreTest
graph {
 properties {bUsedWithStaticMesh=false, materialDomain=MD_UI, pixelDepthOffsetMode=PDOM_Legacy, usedShadingModels=Default Lit}
 nodes [
  n1 = MaterialFunctionCall() {materialFunction=&GetUserInterfaceUV}
  n2 = MaterialFunctionCall(Float2=@n1.Pixel Size) {materialFunction=&BreakOutFloat2Components}
  n3 = AppendVector(A=Multiply(A=Divide(A=@n2.G, B=@n2.G)) {constB=2}, B=Constant() {r=1})
  n4 = MaterialFunctionCall() {materialFunction=&GetUserInterfaceUV}
  n5 = MaterialFunctionCall(Float2=@n4.Pixel Size) {materialFunction=&BreakOutFloat2Components}
  n6 = AppendVector(A=Multiply(A=Divide(A=@n5.G, B=@n5.G)) {constB=2}, B=Constant() {r=1})
  StaticSwitchParameter(
   True=TextureSample(UVs=MaterialFunctionCall(UVs=@n1.Normalized UV, Texture Scale=@n3) {materialFunction=&ScaleUVsByCenter}.UVs) {texture=&/T_UI_StaminaBase_Purple}.RGB
   False=TextureSample(UVs=MaterialFunctionCall(UVs=@n4.Normalized UV, Texture Scale=@n6) {materialFunction=&ScaleUVsByCenter}.UVs) {texture=&/T_UI_StaminaBase}.RGB
  ) {defaultValue=true}
 ]
}
